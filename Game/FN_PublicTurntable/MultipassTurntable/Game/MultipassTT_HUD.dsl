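; HUD for the multipass turntable that draws a screen text line and a character image for scale comparison.
; Each draw pass runs both drawing functions, using values set through setScreenText and setDoDrawScale.
; When enabled, the square image sits at the left edge, sized by the screen distance between the bounds center and a point 192 units above it.
; Text is anchored to the viewport bottom and drawn twice for a black drop shadow.
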(event EventBeginPlay)

(event Collision|EventActorBeginOverlap
  (params (OtherActor :type Actor%20Object%20Reference)))

(event EventTick
  (params (DeltaSeconds :type Float%20%28single-precision%29)))

(event EventReceiveDrawHUD
  (params (SizeX :type Integer) (SizeY :type Integer))
  (CallFunction|ValidateScaleCompareImage)
  (CallFunction|DrawScreenText))

(fn drawScreenText
  (bind (_outwidth _outheight) (HUD|GetTextSize (Variables|Default|GetscreenText) (Variables|Default|GetscreenTextFont) (Variables|Default|GetscreenTextScale)))
  (bind (_returnvalue_x _returnvalue_y) (Math|Vector2D|BreakVector2D (Viewport|GetViewportSize)))
  (if (Utilities|String|NotEqualExactly(String) (Variables|Default|GetscreenText))
    (LocalVariables|Setl_screenY (- (- _returnvalue_y _outheight) 5.0))
    (HUD|DrawText (Variables|Default|GetscreenText) "(R=0,G=0,B=0,A=1)" (+ (LocalVariables|Getl_screenX) 2.0) (+ (LocalVariables|Getl_screenY) 2.0) (Variables|Default|GetscreenTextFont) (Variables|Default|GetscreenTextScale))
    (HUD|DrawText (Variables|Default|GetscreenText) (Variables|Default|GetscreenTextColor) (LocalVariables|Getl_screenX) (LocalVariables|Getl_screenY) (Variables|Default|GetscreenTextFont) (Variables|Default|GetscreenTextScale))))

(fn setDoDrawScale
  (params (drawScaleTexture :type Boolean) (boundsCenter :type Vector))
  (Variables|Default|SetdoDrawScaleTexture drawScaleTexture)
  (Variables|Default|SetcurrentBoundsCenter boundsCenter))

(fn setScreenText
  (params (string :type String) (screenTextColor :type Linear%20Color%20Structure :default "(R=1.000000,G=0.239216,B=0.039216,A=1.000000)") (screenTextScale :type Float%20%28double-precision%29 :default 1.5) (screenTextFont :type Font%20Object%20Reference))
  (Variables|Default|SetscreenText string)
  (Variables|Default|SetscreenTextColor screenTextColor)
  (Variables|Default|SetscreenTextScale screenTextScale)
  (Variables|Default|SetscreenTextFont screenTextFont))

(fn ConstructionScript)

(fn validateScaleCompareImage
  (bind _returnvalue (Game|GetPlayerController 0))
  (bind (_loudness _duration) (Audio|BroadcastPlayerImpactatLocation _returnvalue (+ (Variables|Default|GetcurrentBoundsCenter) "0, 0,192.000000")))
  (bind _currentboundscenter (Variables|Default|GetcurrentBoundsCenter))
  (bind (_loudness_2 _duration_2) (Audio|BroadcastPlayerImpactatLocation _returnvalue _currentboundscenter))
  (bind (_returnvalue_x _returnvalue_y) (Math|Vector2D|BreakVector2D (- (LocalVariables|Getl_foot) (/ (LocalVariables|Getl_heightOnScreen) 2.0))))
  (bind (_returnvalue_x_2 _returnvalue_y_2) (Math|Vector2D|BreakVector2D (Math|Vector2D|MakeVector2D 2.0 _returnvalue_y)))
  (if (Variables|Default|GetdoDrawScaleTexture)
    (LocalVariables|Setl_foot (Math|Vector2D|MakeVector2D _loudness_2 _duration_2))
    (LocalVariables|Setl_head (Math|Vector2D|MakeVector2D _loudness _duration))
    (LocalVariables|Setl_heightOnScreen (Math|Vector2D|Vector2DLength (- (LocalVariables|Getl_head) (LocalVariables|Getl_foot))))
    (HUD|DrawTexture "/Game/FN_PublicTurntable/MultipassTurntable/Materials/EnvCharacterImage.EnvCharacterImage" _returnvalue_x_2 _returnvalue_y_2 (LocalVariables|Getl_heightOnScreen) (LocalVariables|Getl_heightOnScreen) 0.0 0.0 1.0 1.0 "(R=0.350314,G=0.470501,B=1.000000,A=1.000000)")))
